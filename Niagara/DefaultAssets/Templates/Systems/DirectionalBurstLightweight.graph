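# DirectionalBurstLightweight is a Niagara system template in the default assets, named for a lightweight directional burst effect.
# It holds a single emitter, DirectionalBurst, with SystemState as the only listed module.
# No renderers, data interfaces, dynamic inputs or user variables are listed, so the burst's spawning, motion and look are not visible here.
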NiagaraSystem /Niagara/DefaultAssets/Templates/Systems/DirectionalBurstLightweight
dependencies {usedDataInterfaces [], usedDynamicInputs [], usedModules [&SystemState]}
emitters {DirectionalBurst {rendererClasses [], renderers []}}
summary {systemName=DirectionalBurstLightweight, emitters [{emitterName=DirectionalBurst, rendererClasses []}], userVariables []}
