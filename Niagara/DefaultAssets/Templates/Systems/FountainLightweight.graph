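# FountainLightweight is a Niagara system template from the default assets, holding a single emitter named Fountain.
# Its only listed module is SystemState.
# The Fountain emitter lists no renderers, and the system exposes no user variables, data interfaces or dynamic inputs.
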NiagaraSystem /Niagara/DefaultAssets/Templates/Systems/FountainLightweight
dependencies {usedDataInterfaces [], usedDynamicInputs [], usedModules [&SystemState]}
emitters {Fountain {rendererClasses [], renderers []}}
summary {systemName=FountainLightweight, emitters [{emitterName=Fountain, rendererClasses []}], userVariables []}
